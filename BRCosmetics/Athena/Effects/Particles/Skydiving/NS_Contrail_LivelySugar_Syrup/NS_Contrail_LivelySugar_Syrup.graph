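# Niagara skydiving contrail cosmetic that trails two swinging chains with handles and guns, plus bullets, diamonds and gold bars, from the player's skeleton.
# Ribbon_ChainPositions computes two 40-point strands from sockets, owner velocity and sine wobble, which link and handle/gun mesh emitters read to follow.
# Sparkles track Diamond_Gold_Bars particles, using hand-placed offset tables rotated by each mesh's orientation.
# Glide lean, dive and front-end inputs adjust motion, and time-of-day color tints visuals.
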NiagaraSystem /BRCosmetics/Athena/Effects/Particles/Skydiving/NS_Contrail_LivelySugar_Syrup/NS_Contrail_LivelySugar_Syrup
dependencies {
 usedDataInterfaces [
  &NiagaraDataInterfaceArrayFloat3
  &NiagaraDataInterfaceCurve
  &NiagaraDataInterfaceArrayFloat
  &NiagaraDataInterfaceParticleRead
  &NiagaraDataInterfaceArrayInt32
  &NiagaraDataInterfaceSpriteRendererInfo
  &NiagaraDataInterfaceMeshRendererInfo
  &NiagaraDataInterfaceSkeletalMesh
 ]
 usedDynamicInputs [&FNTimeOfDayColor]
 usedModules [
  &self:SystemSpawnScript.SystemScriptSource.SystemScriptGraph.NiagaraNodeAssignment_0.SetVariables_#1
  &SystemState
  /Game/Effects/Niagara/Modules/Contrails/FN_Contrail_Setup
  &FN_Integrator
  &self:SystemSpawnScript.SystemScriptSource.SystemScriptGraph.NiagaraNodeAssignment_1.SetVariables_#2
  &self:SystemSpawnScript.SystemScriptSource.SystemScriptGraph.NiagaraNodeAssignment_3.SetVariables_#3
 ]
}
emitters {
 Bullets {
  emitterUpdateScript [
   &EmitterState()
   &SpawnRate(SpawnRate=15, Use Spawn Probability=false)
  ]
  particleSpawnScript [
   &PartitionParticles()
   &InitializeParticle(
    Lifetime Max=1.5
    Lifetime Mode=::NewEnumerator1
    Mass Max=1.5
    Mass Mode=::NewEnumerator3
    Mesh Scale Mode=::NewEnumerator3
    Mesh Uniform Scale=0.5
    Sprite Rotation Mode=::NewEnumerator2
    Sprite Size Mode=::NewEnumerator3
    Uniform Sprite Size=0.75
    UsePositionOffset=false
    Write Material Random=true
   )
   /SetVariables(Particles.VisibilityTag=&SelectIntFromArray(
     Direct Array Index=&RandomRangeInt(Minimum=0, Maximum=2)
     Int Selection Array=DataInterface() {propertyValues {IntData [0,1,1]}}
    )) {isSetParameters=true}
   &ShapeLocation(MaskSpawnGroup=false, Override Seed=false, Sphere Radius=20, UseOwnerScale=true)
   &SkeletalMeshLocation(
    Bone / Socket Index=$Particles.PartitionParticles.Partition
    MaskSpawnGroup=false
    Mesh Sampling Type=::NewEnumerator5
    Orientation Sampling=::NewEnumerator1
    Sampled Position Offset=$Particles.ShapeLocation.ShapePosition
    Skeletal Mesh=$System.Skeletal Mesh
    Socket Sampling Mode=::NewEnumerator5
   )
   &InitialMeshOrientation(Use Rotation Vector=false)
   &SubUVAnimation(
    Evaluation Type=::NewEnumerator0
    "Sprite Renderer "=DataInterface() {propertyValues {SpriteRenderer=&self:Bullets_0.NiagaraSpriteRendererProperties_0}}
    SubUV Animation Mode:ENiagara_SubUVLookupModeV2=::NewEnumerator2
    UseEndFrame=false
    UseStartFrame=false
   )
   &AddVelocity(
    Velocity=&FN_FrontEndVector3Modifier(
     Gameplay=$System.FN_Contrail_Setup.Direction Vector
     Front End=&Multiply_Vector(
      A=$System.FN_Contrail_Setup.Direction Vector
      B=&MergeHelper_VectorSwitch(0=<-1 1 1>, 1=<1 1 1>, 2=?, 3=?, Index=$Particles.PartitionParticles.Partition, Num=::NewEnumerator0)
     )
    )
    Velocity Speed Scale=&Lerp_Float(A=-150, B=-300, Alpha=$User.ParaGlideLeanAlpha)
   )
  ]
  particleUpdateScript [
   &ParticleState()
   &SpriteRotationRate(Rotation Rate=&RandomRangeFloat(Minimum=-360, Maximum=360))
   &UpdateMeshOrientation(Rotation Vector=&RandomRangeVector(Minimum=<0.2 0.2 0.2>, Maximum=<1 1 1>))
   &ScaleMeshSize(Scale Factor=&Multiply_VectorByFloat(
     Vector=<1 1 1>
     Float=&FloatFromCurve(FloatCurve=DataInterface() {propertyValues {Curve {keys [<at 0 1>,<at 0.7 1>,<at 1 0>]}}})
    ))
   &ScaleSpriteSize(
    Uniform Curve Scale=12
    Uniform Curve Sprite Scale=DataInterface() {propertyValues {Curve {keys [<at 0 0>, {tangentMode=RCTM_User, time=0.0546, value=1}, {tangentMode=RCTM_User, time=1, value=0}]}}}
   )
   &ScaleColor(
    Scale Alpha=&FloatFromCurve(FloatCurve=DataInterface() {propertyValues {Curve {keys [<at 0 1>,<at 0.5 1>,<at 1 0>]}}})
    Scale RGB=&Multiply_VectorByFloat(
     Vector=&MakeVectorFromLinearColorRGB(Color=$System.TOD)
     Float=&MakeFloatFromVector(
      Vector=&VectorFromFloat(Value=&MergeHelper_FloatSwitch(0=1, 1=3, 2=?, 3=?, Index=$Particles.VisibilityTag, Num=::NewEnumerator0))
      Channel=::NewEnumerator0
     )
    )
    ScaleA=true
    ScaleRGB=true
   )
   &AccelerationForce(Acceleration=<0 0 300>)
   &Drag(Drag=&RandomRangeFloat(Minimum=0.5, Maximum=1), Use Linear Drag=true, Use Rotational Drag=true)
   &SolveForcesAndVelocity(Clamp Velocity=true, Limit Acceleration=false)
  ]
  rendererClasses [&NiagaraMeshRendererProperties, &NiagaraSpriteRendererProperties]
  renderers [
   {rendererClass=&NiagaraMeshRendererProperties, rendererIndex=0}
   {rendererClass=&NiagaraSpriteRendererProperties, rendererIndex=1}
  ]
 }
 Chain_Handles_Guns {
  emitterUpdateScript [
   &EmitterState(Life Cycle Mode=::NewEnumerator1, Loop Behavior=::NewEnumerator1, Loop Duration=2, UseLoopDelay=false)
   &SpawnBurst_Instantaneous(Spawn Count=2, Use Spawn Probability=false, UseLoopCountLimit=false)
   &SpawnBurst_Instantaneous(Spawn Count=2, Spawn Group=1, Use Spawn Probability=false, UseLoopCountLimit=false) {module=SpawnBurst_Instantaneous001}
  ]
  particleSpawnScript [
   &PartitionParticles(Number Of Partitions=4)
   /SetVariables(Particles.VisibilityTag=$Emitter.SpawnGroup) {isSetParameters=true}
   &InitializeParticle(
    Color=&MergeHelper_ColorSwitch(0=<rgba 3 1.8911 0.3563 1>, 1=<rgba 1 0.7041 0.4767 1>, 2=?, 3=?, Index=$Emitter.SpawnGroup, Num=::NewEnumerator0)
    Lifetime Max=6
    Lifetime Min=3
    Lifetime Mode=::NewEnumerator1
    Mesh Scale Mode=::NewEnumerator3
    Mesh Uniform Scale=&MergeHelper_FloatSwitch(0=1.2, 1=1, 2=?, 3=?, Index=$Emitter.SpawnGroup, Num=::NewEnumerator0)
    Sprite Rotation Mode=::NewEnumerator2
    Sprite Size Mode=::NewEnumerator1
    Sprite UV Mode=::NewEnumerator3
    Uniform Sprite Size Max=20
    Uniform Sprite Size Min=15
    UsePositionOffset=false
    Write Material Random=true
   )
   &CameraOffset(Camera Offset Amount=5)
   &InitialMeshOrientation(
    Mesh Orientation Mode=::NewEnumerator6
    Orientation Quaternion:Quat4f=/Game/Effects/Niagara/Modules/MergeEmitters/MergeHelper_QuatSwitch(
     0:Quat4f=<0 0 0 1>
     1:Quat4f=<0 0 0 1>
     2:Quat4f=<0 0 -0.7 1>
     3:Quat4f=<0 0 0.7 1>
     4:Quat4f=?
     5:Quat4f=?
     6:Quat4f=?
     7:Quat4f=?
     8:Quat4f=?
     9:Quat4f=?
     Index:NiagaraInt32=$Particles.PartitionParticles.Partition
     Num:MergeHelper_SwitcherEnum=::NewEnumerator2
    )
    Use Rotation Vector=false
   )
  ]
  particleUpdateScript [
   &ParticleState(Kill Particles When Lifetime Has Elapsed=false, Loop Particles Lifetime=true)
   &self:ScratchModule(
    New Particle Attribute Reader=DataInterface() {propertyValues {EmitterBinding {bindingMode=Other, emitterName=Ribbon_ChainPositions}}}
    Particle Index:NiagaraInt32=&SelectIntFromArray(
     Direct Array Index=&ReturnExecIndex()
     Int Selection Array=DataInterface() {propertyValues {IntData [1,41,39,79]}}
    )
   ) {module=ScratchModule}
   &InterpolateOverTime(
    Rate Of Change:NiagaraFloat=&MergeHelper_FloatSwitch(0=15, 1=100, 2=?, 3=?, Index=$Emitter.SpawnGroup, Num=::NewEnumerator0)
    Target Vector Value:Vector3f=&Multiply_Vector(A=$Output.ScratchModule.Direction Vector, B=<1 1 1>)
    Variable Type=::NewEnumerator2
   )
   &UpdateMeshOrientation(
    Facing Direction=&Lerp_Vector(A=$Particles.InterpolateOverTime.InterpolateOverTime_V3_.Moving Average, B=<0 0 1>, Alpha=0.6)
    Facing Interpolation=false
    Orientation Method=::NewEnumerator1
    Override Quat=false
    TransformFacing=false
    TransformUp=false
    Up Direction:Vector3f=<0 1 1>
    Write to Orientation:ENiagara_AttributeSamplingApplyOutput=::NewEnumerator1
   )
   &UpdateMeshOrientation(
    Facing Direction=&Lerp_Vector(A=$Particles.InterpolateOverTime.InterpolateOverTime_V3_.Moving Average, B=<1 1 1>, Alpha=0)
    Facing Interpolation=false
    Facing Interpolation Rate=3
    Orientation Method=::NewEnumerator1
    Override Quat=false
    Reference Axis Vector:Vector3f=&MergeHelper_VectorSwitch(0=<0 0 0>, 1=<0 0 0>, 2=<-0.2 0 1>, 3=<0.5 0 1>, Index=$Particles.PartitionParticles.Partition, Num=::NewEnumerator2)
    Stabilization Mode:ENiagara_OrientVectorCount=::NewEnumerator0
    TransformFacing=false
    Write to Orientation:ENiagara_AttributeSamplingApplyOutput=::NewEnumerator1
   ) {module=UpdateMeshOrientation001}
   /SetVariables(Particles.HandleOrientation:Quat4f=&MultiplyQuaternion(Quaternion A=$StackContext.UpdateMeshOrientation.PreviousFrameQuat, Quaternion B=$Output.UpdateMeshOrientation.MeshOrientation)) {isSetParameters=true}
   /SetVariables(Particles.GunOrientation:Quat4f=&MultiplyQuaternion(Quaternion A=$StackContext.UpdateMeshOrientation001.PreviousFrameQuat, Quaternion B=$Output.UpdateMeshOrientation001.MeshOrientation)) {isSetParameters=true}
   &ScaleColor(Scale Mode=1, Scale RGBA=&MakeVector4_FromLinearColor(LinearColor=$System.TOD), ScaleRGBA=true)
  ]
  rendererClasses [&NiagaraMeshRendererProperties]
  renderers [
   {rendererClass=&NiagaraMeshRendererProperties, rendererIndex=0}
   {rendererClass=&NiagaraMeshRendererProperties, rendererIndex=1}
  ]
 }
 Chain_Mesh_Links {
  emitterUpdateScript [
   &EmitterState(Life Cycle Mode=::NewEnumerator1, Loop Behavior=::NewEnumerator1, Loop Duration Mode=::NewEnumerator1, UseLoopDelay=false)
   &SpawnBurst_Instantaneous(Spawn Count=80, Use Spawn Probability=false, UseLoopCountLimit=false)
  ]
  particleSpawnScript [
   &PartitionParticles(
    Normalized Index In Partition Includes One=true
    Normalized Index Includes One=true
    Particles Per Partition=40
    Partition Mode:ENiagara_PartitionMode_v2=::NewEnumerator1
   )
   /SetVariables(Particles.MeshIndex=&SelectIntFromArray(
     Direct Array Index=&ReturnExecIndex()
     Int Selection Array=DataInterface() {propertyValues {IntData [0,0,0,0,0,0,0,0,0,0,0,0,0,0,0,0,0,0,0,0,0,1,1,1,1,1,1,1,1,0,0,0,0,0,0,0,0,0,0,0,0,0,0,0,0,0,0,0,0,0,0,0,1,1,1,1,1,1,0,0,0,0,0,0,0,0,4,5,2,3,0,0,0,0,0,0,0,0,0,0]}}
    )) {isSetParameters=true}
   &InitializeParticle(
    Color=<rgba 3 2.1953 0.836 1>
    Lifetime Max=6
    Lifetime Min=3
    Lifetime Mode=::NewEnumerator1
    Mesh Renderer Array Visibility Mode=::NewEnumerator1
    Mesh Renderer Info=DataInterface() {propertyValues {MeshRenderer=&self:Chain_Mesh_Links_1.NiagaraMeshRendererProperties_0}}
    Mesh Renderer Mesh Index:NiagaraInt32=$Particles.MeshIndex
    Mesh Scale=&VectorFromFloat(Value=1)
    Mesh Scale Mode=::NewEnumerator4
    Sprite Rotation Mode=::NewEnumerator2
    Sprite Size Mode=::NewEnumerator1
    Sprite UV Mode=::NewEnumerator3
    Uniform Sprite Size Max=20
    Uniform Sprite Size Min=15
    UsePositionOffset=false
    Write Material Random=true
   )
   &CameraOffset(Camera Offset Amount=5)
   &InitialMeshOrientation(
    Mesh Orientation Mode=::NewEnumerator6
    Orientation Quaternion:Quat4f=/Niagara/DynamicInputs/TypeConversions/MakeCustomQuaternionFromBool(
     Bool:NiagaraBool=&SetBoolByIntComparison(A=$Particles.MeshIndex, B=1, Comparison Type=::NewEnumerator4)
     True Quat:Quat4f=&MergeHelper_QuatSwitch(
      0:Quat4f=<0.5 0 -0.75 1>
      1:Quat4f=&MakeCustomQuaternionFromBool(
       Bool:NiagaraBool=&Boolean_OrOperation(Bool1=$User.bIsFrontEnd, Bool2=$User.bIsFrontEndPreview)
       True Quat:Quat4f=<0.5 0.75 0 1>
       False Quat:Quat4f=<0.5 0 0.7 0>
      )
      2:Quat4f=?
      3:Quat4f=?
      4:Quat4f=?
      5:Quat4f=?
      6:Quat4f=?
      7:Quat4f=?
      8:Quat4f=?
      9:Quat4f=?
      Index:NiagaraInt32=$Particles.PartitionParticles.Partition
      Num:MergeHelper_SwitcherEnum=::NewEnumerator0
     )
     False Quat:Quat4f=<0.5 0 1 1>
    )
    Rotation=&MakeCustomVectorFromBool(
     Bool=&SetBoolByIntComparison(A=$Particles.MeshIndex, B=1, Comparison Type=::NewEnumerator4)
     True Vector=&MergeHelper_VectorSwitch(0=<0 0 0>, 1=<0 0 0>, 2=?, 3=?, Index=$Particles.PartitionParticles.Partition, Num=::NewEnumerator0)
     False Vector=<0 0 0>
    )
    Use Rotation Vector=true
   )
  ]
  particleUpdateScript [
   &ParticleState(Kill Particles When Lifetime Has Elapsed=false)
   &ScaleMeshSize(Scale Factor=&VectorFromFloat(Value=&SelectFloatFromArray(
      Direct Array Index=&ReturnExecIndex()
      Float Selection Array=DataInterface() {propertyValues {FloatData [1,1,1,1,1,1,1,1,1,1,1,1,1,1,1,1,1,1,1,1,1,1,1,1,1,1,1,1,1,1,1,1,1,1,1,1,1,1,1,1,1,1,1,1,1,1,1,1,1,1,1,1,1,1,1,1,1,1,1,1,1,1,1,1,1,0.5,1.5,1.5,0.8,2,0.7,1,1,1,1,1,1,1,1,1]}}
     ))) {module=ScaleMeshSize001}
   &ScaleMeshSize(Scale Factor=&MergeHelper_VectorSwitch(
     0=&Multiply_VectorByFloat(Vector=<1 1 1>, Float=0.5)
     1=<0.9 0.9 0.5>
     2=&VectorFromFloat(Value=1.1)
     3=?
     Index=$Particles.MeshIndex
     Num=::NewEnumerator1
    ))
   &self:ScratchModule(
    New Particle Attribute Reader=DataInterface() {propertyValues {EmitterBinding {bindingMode=Other, emitterName=Ribbon_ChainPositions}}}
    Particle Index:NiagaraInt32=&ReturnExecIndex()
   ) {module=ScratchModule}
   &InterpolateOverTime(Rate Of Change=5, Target Vector Value:Vector3f=$Output.ScratchModule.Direction Vector, Variable Type=::NewEnumerator2)
   &UpdateMeshOrientation(
    Facing Direction=&Lerp_Vector(A=$Output.ScratchModule.Direction Vector, B=<1 1 1>, Alpha=0.6)
    Facing Interpolation=true
    Facing Interpolation Rate=3
    Orientation Method=::NewEnumerator1
    Override Quat=false
    Reference Axis Vector:Vector3f=&MergeHelper_VectorSwitch(
     0=<0 0.25 0>
     1=<0 -0.25 0>
     2=?
     3=?
     Index=/Niagara/DynamicInputs/Divide/Divide_Int(A:NiagaraInt32=&ReturnExecIndex(), B:NiagaraInt32=40)
     Num=::NewEnumerator0
    )
    Stabilization Mode:ENiagara_OrientVectorCount=::NewEnumerator0
    TransformFacing=false
   )
   &ScaleColor(Scale Mode=1, Scale RGBA=&MakeVector4_FromLinearColor(LinearColor=$System.TOD), ScaleRGBA=true)
  ]
  rendererClasses [&NiagaraMeshRendererProperties]
  renderers [{rendererClass=&NiagaraMeshRendererProperties, rendererIndex=0}]
 }
 Diamond_Gold_Bars {
  emitterUpdateScript [
   &EmitterState()
   &SpawnRate(
    Spawn Probability=&Lerp_Float(A=0.5, B=1, Alpha=$User.ParaGlideLeanAlpha)
    SpawnRate=6
    Use Spawn Probability=true
   )
  ]
  particleSpawnScript [
   &InitializeParticle(
    Lifetime Min=1.4
    Lifetime Mode=::NewEnumerator1
    Mesh Scale Mode=::NewEnumerator1
    Mesh Uniform Scale Max=0.8
    Mesh Uniform Scale Min=0.7
    Sprite Rotation Mode=::NewEnumerator2
    UsePositionOffset=false
    Write Material Random=true
   )
   &SkeletalMeshLocation(
    Bone / Socket Index=4
    MaskSpawnGroup=false
    Mesh Sampling Type=::NewEnumerator5
    Sampled Position Offset=&Lerp_Vector(A=<0 75 -60>, B=<0 0 -60>, Alpha=$System.FN_Contrail_Setup.Dive Value)
    Sampled Position Offset Coordinate Space:ENiagaraSkelMeshTransforms=::NewEnumerator2
    Skeletal Mesh=$System.Skeletal Mesh
    Socket Sampling Mode=::NewEnumerator5
   )
   &ShapeLocation(
    MaskSpawnGroup=false
    Offset Mode=::NewEnumerator0
    Override Seed=false
    Rotation Mode=::NewEnumerator1
    Sphere Radius=70
    Sphere Surface Distribution=1
    UseOwnerScale=true
   )
   &InitialMeshOrientation(Use Rotation Vector=false)
   &AddVelocity(Module.Use Falloff Distance=false, Origin Offset=<0 0 -100>, Override Seed=false, Velocity Mode=::NewEnumerator1, Velocity Speed=200) {module=AddVelocity001}
   &CameraOffset(Camera Offset Amount=-50)
   &AddVelocity(Rotation Coordinate Space=::World, Velocity=<0 0 200>)
  ]
  particleUpdateScript [
   /SetVariables(Particles.MeshIndex=&RandomRangeInt(Minimum=0, Maximum=1)) {isSetParameters=true}
   /SetVariables(Particles.MaterialRandom:NiagaraFloat=&MakeFloatFromInt(Int=$Particles.MeshIndex)) {isSetParameters=true}
   &ParticleState()
   &UpdateMeshOrientation(
    Rotation Rate=&RandomRangeFloat(Minimum=0.5, Maximum=1)
    Rotation Vector=&RandomRangeVector(Minimum=<-1 -1 -1>, Maximum=<1 1 1>)
   )
   &ScaleMeshSize(Scale Factor=&Multiply_VectorByFloat(
     Vector=<1 1 1>
     Float=&FloatFromCurve(
      FloatCurve=DataInterface() {propertyValues {Curve {
         defaultValue=1
         keys [
          <at 0 0.6 RCIM_Cubic>
          {arriveTangent=0.1844, interpMode=RCIM_Cubic, leaveTangent=0.1844, time=0.2228, value=0.8}
          {arriveTangent=-1.0293, interpMode=RCIM_Cubic, leaveTangent=-1.0293, time=0.8, value=0.7475}
          {tangentMode=RCTM_User, time=1, value=0}
         ]
        }}}
      Scale Curve=&MergeHelper_FloatSwitch(
       0=&RandomRangeFloat(Minimum=0.5, Maximum=0.8)
       1=&RandomRangeFloat(Minimum=0.8, Maximum=1)
       2=?
       3=?
       Index=$Particles.MeshIndex
       Num=::NewEnumerator0
      )
     )
    ))
   &Drag(Drag=&RandomRangeFloat(Minimum=1, Maximum=1.5), Use Linear Drag=true, Use Rotational Drag=true)
   &SolveForcesAndVelocity(Clamp Velocity=false, Limit Acceleration=false)
   &ScaleColor(
    Scale Mode=1
    Scale RGBA=&MakeVector4_FromLinearColor(LinearColor=&MultiplyLinearColor(
      Linear Color=&MergeHelper_ColorSwitch(0=<rgba 2 2 2 1>, 1=<rgba 2 1.4067 0.385 1>, 2=?, 3=?, Index=$Particles.MeshIndex, Num=::NewEnumerator0)
      Scale Factor=&MakeVector4_FromLinearColor(LinearColor=$System.TOD)
     ))
    ScaleRGBA=true
   )
  ]
  rendererClasses [&NiagaraMeshRendererProperties]
  renderers [{rendererClass=&NiagaraMeshRendererProperties, rendererIndex=0}]
 }
 Diamonds {
  emitterUpdateScript [
   &EmitterState()
   &SpawnRate(SpawnRate=7, Use Spawn Probability=false)
  ]
  particleSpawnScript [
   &PartitionParticles()
   &InitializeParticle(
    Color=<rgba 0.6 0.63 1 1>
    Lifetime Max=1.5
    Lifetime Mode=::NewEnumerator1
    Mass Max=1.5
    Mass Mode=::NewEnumerator3
    Mesh Scale Mode=::NewEnumerator3
    Mesh Uniform Scale=0.5
    Sprite Rotation Mode=::NewEnumerator2
    Sprite Size Mode=::NewEnumerator1
    Sprite UV Mode=::NewEnumerator3
    Uniform Sprite Size Max=20
    Uniform Sprite Size Min=10
    UsePositionOffset=false
    Write Material Random=true
   )
   &ShapeLocation(MaskSpawnGroup=false, Override Seed=false, Sphere Radius=10, UseOwnerScale=true)
   &SkeletalMeshLocation(
    Bone / Socket Index=$Particles.PartitionParticles.Partition
    MaskSpawnGroup=false
    Mesh Sampling Type=::NewEnumerator5
    Orientation Sampling=::NewEnumerator1
    Sampled Position Offset=$Particles.ShapeLocation.ShapePosition
    Skeletal Mesh=$System.Skeletal Mesh
    Socket Sampling Mode=::NewEnumerator5
   )
   &AddVelocity(
    Velocity=&FN_FrontEndVector3Modifier(
     Gameplay=$System.FN_Contrail_Setup.Direction Vector
     Front End=&Multiply_Vector(
      A=$System.FN_Contrail_Setup.Direction Vector
      B=&MergeHelper_VectorSwitch(0=<-1 1 1>, 1=<1 1 1>, 2=?, 3=?, Index=$Particles.PartitionParticles.Partition, Num=::NewEnumerator0)
     )
    )
    Velocity Speed Scale=&Lerp_Float(A=-150, B=-300, Alpha=$User.ParaGlideLeanAlpha)
   )
   &SubUVAnimation(
    Evaluation Type=::NewEnumerator0
    "Sprite Renderer "=DataInterface() {propertyValues {SpriteRenderer=&self:Diamonds_0.NiagaraSpriteRendererProperties_0}}
    SubUV Animation Mode:ENiagara_SubUVLookupModeV2=::NewEnumerator2
    UseEndFrame=false
    UseStartFrame=false
   )
  ]
  particleUpdateScript [
   &ParticleState()
   &SpriteRotationRate(Rotation Rate=&RandomRangeFloat(Minimum=-360, Maximum=360))
   &SolveForcesAndVelocity(Clamp Velocity=true, Limit Acceleration=false)
   &ScaleSpriteSize(Uniform Curve Sprite Scale=DataInterface() {propertyValues {Curve {keys [<at 0 0 RCIM_Cubic>,<at 0.0546 1 RCIM_Cubic>,<at 1 0 RCIM_Cubic>]}}})
   &ScaleColor(
    Scale Alpha=&FloatFromCurve(FloatCurve=DataInterface() {propertyValues {Curve {keys [<at 0.5 1>,<at 1 0>]}}}, Scale Curve=0.7)
    Scale RGB=&Multiply_VectorByFloat(
     Vector=&MakeVectorFromLinearColorRGB(Color=$System.TOD)
     Float=&MakeFloatFromVector(
      Vector=&VectorFromFloat(Value=&RandomRangeFloat(Minimum=1, Maximum=1.2))
      Channel=::NewEnumerator0
     )
    )
    ScaleA=true
    ScaleRGB=true
   )
   &FN_Sprite_Attributes(
    Camera Offset:NiagaraFloat=&Add_Float(A=$Particles.CameraOffset, B=5)
    Color=<rgba 5 5 5 1>
    Create Camera Offset Attribute=true
    Create Color Attribute=true
    Create Position Attribute=true
    Create Sprite Rotation Attribute=true
    Create SpriteSize Attribute=true
    Module Output:ENiagara_AttributeSamplingApplyOutput=::NewEnumerator1
    Position:NiagaraPosition=&AddVectorToPosition(Position=$Particles.Position, Vector=<0 0 0>)
    SpriteRotation:NiagaraFloat=&RandomRangeFloat(Minimum=200, Maximum=360)
    SpriteSize=&Multiply_Vector2D_ByFloat(
     Float=&FloatFromCurve(
      FloatCurve=DataInterface() {propertyValues {Curve {
         defaultValue=0
         keys [
          {interpMode=RCIM_Cubic, tangentMode=RCTM_User, time=0.15, value=0}
          {interpMode=RCIM_Cubic, tangentMode=RCTM_User, time=0.3, value=1}
          {tangentMode=RCTM_User, time=0.5, value=0}
         ]
        }}}
      CurveIndex=&Multiply_Float(A=$Particles.NormalizedAge, B=&RandomRangeFloat(Minimum=0.8, Maximum=2))
     )
     Vector2D=&Vector2DFromFloat(Value=&RandomRangeFloat(Minimum=15, Maximum=20))
    )
   )
  ]
  rendererClasses [&NiagaraSpriteRendererProperties]
  renderers [
   {rendererClass=&NiagaraSpriteRendererProperties}
   {rendererClass=&NiagaraSpriteRendererProperties, rendererIndex=1}
  ]
 }
 Ribbon_ChainPositions {
  emitterUpdateScript [
   &EmitterState(Life Cycle Mode=::NewEnumerator1, Loop Behavior=::NewEnumerator1, Loop Duration Mode=::NewEnumerator1, UseLoopDelay=false)
   &SpawnBurst_Instantaneous(
    Spawn Count=&Multiply_Int(A=2, B=40)
    Spawn Probability=HlslExpression() {hlslExpression="Emitter.LoopCount == 0"}
    Use Spawn Probability=false
    UseLoopCountLimit=false
   ) {module=SpawnBurst_Instantaneous001}
  ]
  particleSpawnScript [
   &PartitionParticles(
    Normalized Index In Partition Includes One=true
    Normalized Index Includes One=true
    Particles Per Partition=40
    Partition Mode:ENiagara_PartitionMode_v2=::NewEnumerator1
   )
   &InitializeParticle(
    Lifetime=5
    Ribbon Width=8
    Ribbon Width Mode=::NewEnumerator1
    Sprite Size Mode=::NewEnumerator3
    Uniform Sprite Size=2
    UsePositionOffset=false
    Write Material Random=true
   ) {module=InitializeParticle001}
  ]
  particleUpdateScript [
   &ParticleState(Kill Particles When Lifetime Has Elapsed=false)
   &SkeletalMeshLocation(
    Bone / Socket Index=&ModuloInt(A=$Particles.PartitionParticles.Partition, B=2)
    MaskSpawnGroup=false
    Mesh Sampling Type=::NewEnumerator5
    Orientation Sampling=::NewEnumerator1
    Sampled Position Offset=&Lerp_Vector(
     A=<0 -2 20>
     B=&FN_FrontEndVector3Modifier(Gameplay=<0 5 20>, Front End=<0 -3 18>)
     Alpha=$User.ParaGlideLeanAlpha
    )
    Sampled Position Offset Coordinate Space:ENiagaraSkelMeshTransforms=::NewEnumerator1
    Skeletal Mesh=$System.Skeletal Mesh
    Socket Sampling Mode=::NewEnumerator5
    Use Sampled Mesh Attribute as Ribbon ID=false
   )
   /SetVariables(Particles.Position=&AddVectorToPosition(
     Position=$Particles.Position
     Vector=&Multiply_VectorByFloat(
      Vector=&Add_Vector(
       A=&Multiply_VectorByFloat(
        Vector=&NormalizeVector(
         Vector To Normalize=&FN_FrontEndVector3Modifier(
          Gameplay=&TransformVector(Vector=$Engine.Owner.Velocity, Source Space=::World, Destination Space=::Simulation)
          Front End=$System.FrontendVelocity
         )
         Safe Normalize=true
         Safe Fallback Threshold=0
         Safe Fallback Vector=<0 0 0.001>
        )
        Float=-1
       )
       B=&Multiply_VectorByFloat(
        Vector=&TransformVector(
         Vector=&SelectVectorFromArray(
          Direct Array Index=$Particles.PartitionParticles.Partition
          Direct Array Mode=::NewEnumerator1
          Vector Selection Array=DataInterface() {propertyValues {FloatData [<2 -2 0>, <-2 0 0>]}}
         )
         Destination Space=::Simulation
        )
        Float=&Lerp_Float(A=0.05, B=0.15, Alpha=$User.ParaGlideLeanAlpha)
       )
      )
      Float=&FloatFromCurve(
       FloatCurve=DataInterface() {propertyValues {Curve {defaultValue=0, keys [{tangentMode=RCTM_User, time=0, value=0}, {tangentMode=RCTM_User, time=1, value=1}]}}}
       CurveIndex=$Particles.PartitionParticles.Normalized Index In Partition
       Scale Curve=&SelectFloatFromArray(Direct Array Index=$Particles.PartitionParticles.Partition, Float Selection Array=DataInterface() {propertyValues {FloatData [105,105]}})
      )
     )
    )) {isSetParameters=true}
   /SetVariables(Particles.Position=&AddVectorToPosition(
     Position=$Particles.Position
     Vector=&Multiply_VectorByFloat(
      Vector=/Niagara/DynamicInputs/Vectors/RotateVector(
       Vector To Rotate:Vector3f=<10 0 3>
       Yaw:NiagaraFloat=&Add_Float(
        A=$Particles.Age
        B=&Multiply_Float(A=&MakeFloatFromInt(Int=$Particles.PartitionParticles.Partition), B=2)
       )
       Pitch:NiagaraFloat=0
       Roll:NiagaraFloat=0
       Rotation Rate:NiagaraFloat=&SelectFloatFromArray(Direct Array Index=$Particles.PartitionParticles.Partition, Float Selection Array=DataInterface() {propertyValues {FloatData [140,140]}})
      )
      Float=&Multiply_Float(
       A=&Sine(
        Normalized Angle=&Add_Float(
         A=&Multiply_Float(
          A=$Emitter.Age
          B=&SelectFloatFromArray(Direct Array Index=$Particles.PartitionParticles.Partition, Float Selection Array=DataInterface() {propertyValues {FloatData [-2.5,-2.5]}})
         )
         B=$Particles.PartitionParticles.Normalized Index In Partition
        )
        Period=0.9
        Scale=1
        Bias=0
       )
       B=&FloatFromCurve(
        FloatCurve=DataInterface() {propertyValues {Curve {defaultValue=0, keys [{tangentMode=RCTM_User, time=0, value=0}, {tangentMode=RCTM_User, time=1, value=1}]}}}
        CurveIndex=$Particles.PartitionParticles.Normalized Index In Partition
       )
      )
     )
    )) {isSetParameters=true}
   &ScaleColor(ScaleA=true, ScaleRGB=true) {module=ScaleColor001}
  ]
  rendererClasses [&NiagaraSpriteRendererProperties]
  renderers [{rendererClass=&NiagaraSpriteRendererProperties}]
 }
 Sparkles {
  emitterUpdateScript [
   &EmitterState()
   &SpawnParticlesFromOtherEmitter(
    Attribute Reader=DataInterface() {propertyValues {EmitterBinding {bindingMode=Other, emitterName=Diamond_Gold_Bars}}}
    Spawn Rate=5
    Use Maximum Age=false
    Use Minimum Age=false
   )
  ]
  particleSpawnScript [
   &InitializeParticle(
    Color Mode=::NewEnumerator0
    Lifetime Max=0.5
    Lifetime Min=0.3
    Lifetime Mode=::NewEnumerator1
    Sprite Size Mode=::NewEnumerator1
    Uniform Sprite Size Max=30
    Uniform Sprite Size Min=20
    UsePositionOffset=false
    Write Material Random=true
   )
   &SampleParticlesFromOtherEmitter(
    Mesh Orientation Sampling:ENiagara_DirectReadSamplingMode=::NewEnumerator1
    Mesh Scale Sampling:ENiagara_DirectReadSamplingMode=::NewEnumerator1
    Particle ID Sampling=::NewEnumerator1
    Position Sampling:ENiagara_DirectReadSamplingMode=::NewEnumerator1
   )
   &CameraOffset(Camera Offset Amount=5)
  ]
  particleUpdateScript [
   &ParticleState()
   &SolveForcesAndVelocity(Clamp Velocity=false, Limit Acceleration=false)
   /Niagara/Modules/AttributeReader/UpdateParticlesFromOtherEmitter(
    Age Sampling:ENiagara_DirectReadSamplingMode=::NewEnumerator1
    Attribute Reader=DataInterface() {propertyValues {EmitterBinding {bindingMode=Other, emitterName=Diamond_Gold_Bars}}}
    Lifetime Sampling:ENiagara_DirectReadSamplingMode=::NewEnumerator1
    Material Random Sampling:ENiagara_DirectReadSamplingMode=::NewEnumerator1
    Mesh Orientation Sampling:ENiagara_DirectReadSamplingMode=::NewEnumerator1
    Mesh Scale Sampling:ENiagara_DirectReadSamplingMode=::NewEnumerator1
    Position Sampling:ENiagara_DirectReadSamplingMode=::NewEnumerator1
   )
   /SetVariables(Particles.TestPos:NiagaraPosition=&AddVectorToPosition(
     Position=$Output.UpdateParticlesFromOtherEmitter.SampledPosition
     Vector=&RotateVectorByQuaternion(
      Vector To Rotate=&Multiply_Vector(
       A=&MergeHelper_VectorSwitch(
        0=&SelectVectorFromArray(
         Array Sampling Mode=::NewEnumerator0
         Direct Array Index=?
         Vector Selection Array=DataInterface() {propertyValues {FloatData [
            <3.2558 13.2775 31.1179>
            <-18.3931 -0.2826 26.9009>
            <0.5753 -16.7952 28.5188>
            <17.9486 -2.733 26.2871>
            <-7.6926 -3.6271 32.1551>
            <-2.8864 -4.724 32.1551>
            <14.5682 -12.4065 23.0577>
            <2.3319 -19.2493 23.6312>
            <4.0535 -18.8927 24.0319>
            <-3.3654 -16.1118 20.585>
            <-11.9214 -0.6439 14.5917>
            <-16.2036 10.9773 24.2002>
            <-1.3077 10.6301 13.4501>
            <3.1947 10.1479 13.7867>
            <16.7458 7.703 23.2867>
            <-16.0005 -3.2425 20.2133>
            <-15.1167 -9.5235 21.7493>
            <-7.7888 5.2742 11.3689>
            <16.6173 9.9367 24.0226>
            <11.3917 -8.2696 16.6742>
            <1.7117 -12.6542 15.7402>
            <9.5074 -11.9021 17.9735>
            <5.4809 -10.0416 13.379>
            <11.303 -15.9199 23.7493>
            <-2.6834 -7.2095 8.8835>
            <-5.2057 -16.1814 21.3305>
            <4.3218 11.1681 32.7919>
            <-0.6025 4.1299 33.0464>
            <-4.4625 -1.1131 33.0464>
            <18.1714 -2.0441 27.1535>
            <15.2842 5.8651 29.6192>
            <-1.3191 19.9476 25.3232>
            <-12.4764 5.9203 32.0979>
           ]}}
         Random Seed=0
        )
        1=&SelectVectorFromArray(
         Array Sampling Mode=::NewEnumerator0
         Direct Array Index=?
         Vector Selection Array=DataInterface() {propertyValues {FloatData [
            <1.9147 2.064 7.4649>
            <1.6755 17.32 7.183>
            <-2.6971 -1.1302 0.139>
            <-3.1151 13.4962 7.8339>
            <5.6412 7.1221 1.6853>
            <-5.6422 5.655 1.8374>
            <5.4618 12.9183 3.7645>
            <5.976 14.9187 1.8138>
            <4.2062 0.4937 7.1906>
            <-4.2187 5.6329 7.3364>
            <-0.4244 14.771 0.1677>
           ]}}
         Random Seed=0
        )
        2=?
        3=?
        Index=&MakeIntFromFloat(Float=$Output.UpdateParticlesFromOtherEmitter.SampledMaterialRandom)
        Num=::NewEnumerator0
       )
       B=&Multiply_VectorByFloat(Vector=$Output.UpdateParticlesFromOtherEmitter.SampledScale, Float=0.9)
      )
      Quaternion=$Output.UpdateParticlesFromOtherEmitter.SampledMeshOrientation
     )
    )) {isSetParameters=true}
   &ScaleSpriteSize(Uniform Curve Sprite Scale=DataInterface() {propertyValues {Curve {
       defaultValue=0
       keys [
        <at 0 0.3>
        {tangentMode=RCTM_User, time=0.2, value=1}
        {arriveTangent=-0.0405, interpMode=RCIM_Cubic, leaveTangent=-0.0405, tangentMode=RCTM_User, time=0.7009, value=1}
        {arriveTangent=-8.0089, interpMode=RCIM_Cubic, leaveTangent=-8.0089, tangentMode=RCTM_User, time=1, value=0}
       ]
      }}})
   &ScaleColor(
    Scale Alpha=&FloatFromCurve(FloatCurve=DataInterface() {propertyValues {Curve {
        defaultValue=1
        keys [
         {interpMode=RCIM_Cubic, tangentMode=RCTM_User, time=0, value=0}
         {interpMode=RCIM_Cubic, tangentMode=RCTM_User, time=0.5, value=1}
         {tangentMode=RCTM_User, time=1, value=0}
        ]
       }}})
    Scale RGB=&MakeVectorFromLinearColorRGB(Color=$System.TOD)
    ScaleA=true
    ScaleRGB=true
   )
   &KillParticles(Kill Particles=&SetBoolByFloatComparison(
     A=$Output.UpdateParticlesFromOtherEmitter.SampledAge
     B=&Subtract_Float(A=$Output.UpdateParticlesFromOtherEmitter.SampledLifetime, B=0.1)
    ))
  ]
  rendererClasses [&NiagaraSpriteRendererProperties]
  renderers [{rendererClass=&NiagaraSpriteRendererProperties}]
 }
}
summary {
 systemName=NS_Contrail_LivelySugar_Syrup
 emitters [
  {emitterName=Ribbon_ChainPositions, rendererClasses [&NiagaraSpriteRendererProperties]}
  {emitterName=Diamond_Gold_Bars, rendererClasses [&NiagaraMeshRendererProperties]}
  {emitterName=Sparkles, rendererClasses [&NiagaraSpriteRendererProperties]}
  {emitterName=Bullets, rendererClasses [&NiagaraMeshRendererProperties, &NiagaraSpriteRendererProperties]}
  {emitterName=Diamonds, rendererClasses [&NiagaraSpriteRendererProperties]}
  {emitterName=Chain_Mesh_Links, rendererClasses [&NiagaraMeshRendererProperties]}
  {emitterName=Chain_Handles_Guns, rendererClasses [&NiagaraMeshRendererProperties]}
 ]
 userVariables [
  User.bIsFrontEnd {type=NiagaraBool}
  User.bIsFrontEndPreview {type=NiagaraBool}
  User.ParaGlideLeanAlpha {type=NiagaraFloat, defaultValue=0, description=""}
  User.PlayerSpeed {type=NiagaraFloat, defaultValue=0, description=""}
  User.RibbonIDOffset {type=NiagaraInt32}
 ]
}
